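# Translucent unlit Niagara material for the Majesty pickaxe's golden dust trail, supporting sprites, ribbons and meshes.
# Scalar-parameter lerps blend a panned swing-trail texture, backpack-style mesh dust, two-color glow and mask-line variants.
# Screen-aligned noise creates sand sparkle, colored by a curve atlas and thresholded for opacity, with depth and view-angle fades.
# Dynamic parameters drive trail offset, color change, rotated-UV dust reveal and a UV-weighted world position bend.
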Material /BRCosmetics/Weapons/FORT_Melee/Pickaxe_Majesty_Male/FX/M_Pickaxe_Majesty_Golden_dust_trail
graph {
 properties {
  bEnableResponsiveAA=true
  bUsedWithNiagaraMeshParticles=true
  bUsedWithNiagaraRibbons=true
  bUsedWithNiagaraSprites=true
  blendMode=BLEND_Translucent
  pixelDepthOffsetMode=PDOM_Legacy
  shadingModel=MSM_Unlit
  usedShadingModels=Unlit
 }
 nodes [
  Constant() {r=-0.1}
  Constant() {r=-0.1}
  n1 = Constant() {r=-0.07}
  n2 = MaterialFunctionCall(UVs=TextureCoordinate() {coordinateIndex=1}, "Rotation Angle (0-1)"=@n1) {materialFunction=&CustomRotator}
  n3 = DynamicParameter() {defaultValue=<rgba 1 0 0 1>, paramNames [Swing_Trail_Offset, Mesh_Trail_random_offset, Color_Change, Dust_appear]}
  n4 = Add(
   A=Add(A=ComponentMask(None=@n2.Rotated Values) {g=true}, B=Constant() {r=-0.65}) {constB=-0.3}
   B=LinearInterpolate(Alpha=@n3.Dust_appear) {constA=0.75, constB=-0.6}
  ) {constB=0}
  n5 = MaterialFunctionCall(UVs=TextureCoordinate() {coordinateIndex=1}, "Rotation Angle (0-1)"=@n1) {materialFunction=&CustomRotator}
  n6 = Multiply(A=TextureSample(UVs=Panner(Coordinate=@n5.Rotated Values) {speedX=0.5, speedY=-0.5}) {texture=&T_Inky_Smoke_Tile}.R) {constB=0.25}
  n7 = Multiply(A=Saturate(None=Add(A=@n4, B=@n6) {constB=0})) {constB=20}
  n8 = Saturate(None=@n7)
  n9 = Multiply(A=Saturate(None=@n7), B=Abs(None=Subtract(A=@n8)))
  n10 = Multiply(A=Multiply(A=Multiply(A=@n9))) {constB=500}
  LinearInterpolate(B=Multiply(A=ParticleColor().A, B=Saturate(None=@n10)), Alpha=ScalarParameter() {group=Mask_Line, parameterName=Only_use_mask_Line})
  n11 = DynamicParameter() {desc=Parameters set by Niagara, parameterIndex=2, paramNames [X,Y,Z,WPOIntensity]}
  n12 = Subtract(
   A=MaterialFunctionCall(X=@n11.X, Y=@n11.X, Z=@n11.X) {materialFunction=&MakeFloat3}.Result
   B=ParticlePositionWS()
  )
  n13 = ScalarParameter() {defaultValue=-10, group=WPO, parameterName=Clamp_min}
  n14 = ScalarParameter() {defaultValue=10, group=WPO, parameterName=Clamp_max}
  n15 = Power(Base=ComponentMask(None=TextureCoordinate()) {g=true}, Exp=ScalarParameter() {defaultValue=3, group=WPO, parameterName=WPOBendPower})
  LinearInterpolate(
   A=Constant()
   B=LinearInterpolate(B=Clamp(None=@n12, Min=@n13, Max=@n14) {maxDefault=1000, minDefault=-1000}, Alpha=Multiply(A=@n15, B=@n11.WPOIntensity))
   Alpha=ScalarParameter() {group=WPO, parameterName=Use_Interpolate_WPO}
  )
  MaterialFunctionCall() {materialFunction=&DebugTimeSine}
  n16 = ScalarParameter() {group=Digital_glow, parameterName=Use_simple_23Glow}
  n17 = ParticleColor()
  n18 = VectorParameter() {parameterName=Color01, defaultValue=<rgba 1 0 0.6831 1>}
  n19 = VectorParameter() {parameterName=Color02, defaultValue=<rgba 1 0.5923 0 1>}
  n20 = Power(Base=ComponentMask(None=TextureCoordinate()) {g=true})
  n21 = VectorParameter() {parameterName=Color02, defaultValue=<rgba 1 0.5923 0 1>}
  n22 = OneMinus(None=ComponentMask(None=TextureCoordinate()) {g=true})
  n23 = LinearInterpolate(Alpha=Saturate(None=Power(Base=@n22))) {constA=0.5, constB=2}
  n24 = MaterialFunctionCall(UVs=TextureCoordinate(), "Rotation Angle (0-1)"=Constant() {r=0.1}) {materialFunction=&CustomRotator}
  n25 = DynamicParameter() {defaultValue=<rgba 1 0 0 1>, paramNames [Swing_Trail_Offset, Mesh_Trail_random_offset, Color_Change, Dust_appear]}
  n26 = Add(
   A=Add(A=ComponentMask(None=@n24.Rotated Values) {g=true}, B=Constant() {r=-0.5}) {constB=-0.3}
   B=LinearInterpolate(Alpha=@n25.Color_Change) {constA=0.9, constB=-0.7}
  )
  n27 = Saturate(None=Multiply(A=Saturate(None=@n26)) {constB=5})
  n28 = Multiply(
   A=@n17.RGB
   B=LinearInterpolate(A=LinearInterpolate(A=@n18.RGB, B=@n19.RGB, Alpha=@n20), B=Multiply(A=@n21.RGB, B=@n23), Alpha=OneMinus(None=@n27))
  )
  n29 = Constant() {r=24}
  n30 = TextureCoordinate() {uTiling=8, vTiling=8}
  n31 = TextureSample(UVs=Add(
    A=MaterialFunctionCall!(TextureResolution=@n29) {materialFunction=/Engine/Functions/Engine_MaterialFunctions02/Texturing/ScreenAlignedPixelToPixelUVs}.Result
    B=Multiply(A=@n30) {constB=0.15}
   )) {samplerType=SAMPLERTYPE_Normal, texture=/Game/Environments/World/Backgrounds/Locales/Temperate/Textures/T_WaterNoise}
  n32 = Multiply(A=ConstantBiasScale(None=@n31.RGB), B=Time())
  n33 = CurveAtlasRowParameter(CurveTime=MaterialFunctionCall(Float3=Frac(None=@n32)) {materialFunction=&BreakOutFloat3Components}.R) {
   atlas=&AthenaCurveAtlas01
   curve=/Game/Athena/LinearColorCurves/Curve_Sand_EmissiveColor
   defaultValue=43
   group=Sand
   parameterName=Emissive
  }
  n34 = VectorParameter() {parameterName=Texture_UV_Scale, channelNames {a="", b="", g=V_Scale, r=U_Scale}, defaultValue=<rgba 0.95 1 0 1>}
  n35 = Multiply(A=TextureCoordinate(), B=AppendVector(A=@n34.U_Scale, B=@n34.U_Scale))
  n36 = TextureCoordinate()
  n37 = Multiply(
   A=TextureSample(UVs=Panner(Coordinate=@n36) {speedY=-0.3}) {texture=&T_Inky_Smoke_Tile}.R
   B=ScalarParameter() {defaultValue=0.15, group=Swing_Trail, parameterName=Default_Texture_Noise_Intensity}
  ) {constB=0.15}
  n38 = DynamicParameter() {defaultValue=<rgba 1 0 0 1>, paramNames [Swing_Trail_Offset, Mesh_Trail_random_offset, Color_Change, Dust_appear]}
  n39 = TextureSampleParameter2D(UVs=Panner(Coordinate=Add(A=@n35, B=@n37), Time=@n38.Swing_Trail_Offset) {speedX=0.5}) {
   group=Swing_Trail
   parameterName=Default_Texture
   texture=/BRCosmetics/Weapons/FORT_Melee/Pickaxe_Tower_Sentinel_Female/FX/T_TowerSentinel_trail
  }
  n40 = TextureCoordinate()
  n41 = Multiply(A=Power(Base=ComponentMask(None=@n40) {r=true}))
  n42 = OneMinus(None=ComponentMask(None=TextureCoordinate()) {r=true})
  n43 = Multiply(
   A=Power(Base=Multiply(A=@n39.R, B=@n39.R), Exp=ScalarParameter() {defaultValue=1, group=Swing_Trail, parameterName=Default_Texture_power}) {constExponent=1}
   B=Multiply(A=Saturate(None=@n41), B=Power(Base=@n42) {constExponent=3})
  )
  n44 = ScalarParameter() {defaultValue=1, group=Swing_Trail, parameterName=Texture_intensity}
  n45 = VectorParameter() {parameterName=Texture_UV_Scale, channelNames {a="", b="", g=V_Scale, r=U_Scale}, defaultValue=<rgba 0.95 1 0 1>}
  n46 = Add(A=Multiply(A=TextureCoordinate() {uTiling=2}, B=AppendVector(A=@n45.U_Scale, B=@n45.U_Scale)), B=VertexColor().R)
  n47 = TextureCoordinate() {uTiling=2, vTiling=0.5}
  n48 = VectorParameter() {group=Backpack_Mesh_dust, parameterName=Noise_UV_Scale, channelNames {a="", b="", g=V_Scale, r=U_Scale}, defaultValue=<rgba 1 1 0 1>}
  n49 = AppendVector(A=@n48.U_Scale, B=@n48.U_Scale)
  n50 = Constant()
  n51 = ScalarParameter() {defaultValue=-0.02, group=Backpack_Mesh_dust, parameterName=Noise_Panner}
  n52 = ScalarParameter() {defaultValue=1, group=Backpack_Mesh_dust, parameterName=Speed}
  n53 = Panner(Coordinate=Add(A=Multiply(A=@n47, B=@n49), B=VertexColor().R), Speed=Multiply(A=AppendVector(A=@n50, B=@n51), B=@n52)) {speedY=-0.3}
  n54 = Multiply(
   A=ComponentMask(None=TextureSampleParameter2D(UVs=@n53) {group=Backpack_Mesh_dust, parameterName=Noise_Texture, texture=&T_ShieldCaustics}.RGB) {g=true, r=true}
   B=ScalarParameter() {defaultValue=0.5, group=Backpack_Mesh_dust, parameterName=Disturbance_intensity}
  ) {constB=0.5}
  n55 = Constant2Vector() {g=-0.3}
  n56 = ScalarParameter() {defaultValue=1, group=Backpack_Mesh_dust, parameterName=Speed}
  n57 = TextureSampleParameter2D(UVs=Panner(Coordinate=Add(A=@n46, B=@n54), Speed=Multiply(A=@n55, B=@n56)) {speedY=-0.3}) {group=Backpack_Mesh_dust, parameterName=Default_Mesh_Texture, texture=/Game/Global/Textures/Default/PointSplatter/T_Tiling_Dust}
  n58 = Add(A=Add(A=TextureCoordinate() {vTiling=0.5}, B=VertexColor().R), B=@n54) {constB=0}
  n59 = Multiply(A=Constant2Vector() {g=-0.05}, B=@n52)
  n60 = Multiply(
   A=@n57.R
   B=TextureSample(UVs=Panner(Coordinate=@n58, Speed=@n59) {speedY=0.1}) {texture=/Game/Global/Textures/Default/Caustics/T_Turbulence_Seamless}.R
  )
  n61 = TextureCoordinate()
  n62 = Multiply(A=Power(Base=ComponentMask(None=@n61) {g=true}))
  n63 = OneMinus(None=ComponentMask(None=TextureCoordinate()) {g=true})
  n64 = Multiply(
   A=Power(Base=Multiply(A=@n60, B=@n60), Exp=ScalarParameter() {defaultValue=1, group=Backpack_Mesh_dust, parameterName=Default_Mesh_Texture_power}) {constExponent=1}
   B=Multiply(A=Saturate(None=@n62), B=Power(Base=@n63) {constExponent=4})
  )
  n65 = LinearInterpolate(
   A=Saturate(None=Multiply(A=@n43, B=@n44))
   B=Saturate(None=Multiply(A=@n64) {constB=2000})
   Alpha=ScalarParameter() {group=Backpack_Mesh_dust, parameterName=Use_Backpack_Mesh_trail}
  )
  n66 = ScalarParameter() {parameterName=Use_NoSand}
  n67 = Multiply(A=ParticleColor().RGB, B=LinearInterpolate(A=Multiply(A=@n33) {constB=0.1}, B=@n65, Alpha=@n66))
  MaterialFunctionCall(Emissive=LinearInterpolate(
    A=LinearInterpolate(A=@n67, B=@n28, Alpha=@n16)
    B=ParticleColor().RGB
    Alpha=ScalarParameter() {group=Mask_Line, parameterName=Only_use_mask_Line}
   )) {materialFunction=&MF_FX_EmissiveAdjustments}
  MaterialFunctionCall() {materialFunction=&DebugTimeSine}
  n68 = DynamicParameter() {parameterIndex=1, defaultValue=<rgba 1 0 0 1>, paramNames [R,G,B,Param4]}
  n69 = DotProduct(
   A=@n31.RGB
   B=MaterialFunctionCall(R=@n68.R, G=@n68.R, B=@n68.R, A=@n68.R) {materialFunction=&AppendMany}.RGB
  )
  n70 = Multiply(A=Ceil(None=Saturate(None=@n69)), B=@n65)
  n71 = Add(A=TextureCoordinate(), B=VertexColor().R)
  n72 = Multiply(A=Constant2Vector() {g=-0.5}, B=ScalarParameter() {defaultValue=1, group=Backpack_Mesh_dust, parameterName=Speed})
  n73 = LinearInterpolate(Alpha=TextureSample(UVs=Panner(Coordinate=@n71, Speed=@n72) {speedY=-0.5}) {texture=&T_Inky_Smoke_Tile}.R)
  n74 = Power(Base=OneMinus(None=Fresnel(ExponentIn=@n73) {baseReflectFraction=0, exponent=1})) {constExponent=1}
  n75 = TextureCoordinate() {coordinateIndex=1}
  n76 = Constant() {r=-0.07}
  n77 = Add(
   A=ComponentMask(None=MaterialFunctionCall(UVs=@n75, "Rotation Angle (0-1)"=@n76) {materialFunction=&CustomRotator}.Rotated Values) {g=true}
   B=Constant() {r=-0.55}
  ) {constB=-0.3}
  n78 = DynamicParameter() {defaultValue=<rgba 1 0 0 1>, paramNames [Swing_Trail_Offset, Mesh_Trail_random_offset, Color_Change, Dust_appear]}
  n79 = LinearInterpolate(Alpha=@n78.Dust_appear) {constA=1.1, constB=-0.45}
  n80 = Multiply(A=Saturate(None=Add(A=@n77, B=@n79))) {constB=2}
  n81 = ComponentMask(None=MaterialFunctionCall(UVs=TextureCoordinate() {coordinateIndex=1}, "Rotation Angle (0-1)"=@n76) {materialFunction=&CustomRotator}.Rotated Values) {g=true}
  n82 = Multiply(A=OneMinus(None=Saturate(None=@n80)), B=Saturate(None=Multiply(A=@n81) {constB=10}))
  n83 = Multiply(
   A=Multiply(A=LinearInterpolate(A=@n70, B=@n65, Alpha=@n66), B=ParticleColor().A)
   B=LinearInterpolate(A=Constant() {r=1}, B=Multiply(A=@n74, B=@n82), Alpha=ScalarParameter() {group=Backpack_Mesh_dust, parameterName=Use_Backpack_Mesh_trail})
  )
  n84 = DepthFade(Opacity=@n83, FadeDistance=ScalarParameter() {defaultValue=100, parameterName=Depth_fade})
  n85 = ShadingPathSwitch(Default=@n84, Deferred=@n84, Forward=@n83, Mobile=@n83)
  n86 = MaterialFunctionCall("FadeSensitivity (optional)"=ScalarParameter() {parameterName=MeshFade}) {materialFunction=&FadeBasedOnViewAngle}
  n87 = Multiply(
   A=@n17.A
   B=TextureSample(UVs=TextureCoordinate() {uTiling=-1}) {texture=/BRCosmetics/Accessories/FORT_Backpacks/Backpack_M_MED_Majesty/FX/T_Backpack_Majesty}.R
  )
  n88 = MaterialFunctionCall("FadeSensitivity (optional)"=ScalarParameter() {defaultValue=1.5, group=Digital_glow, parameterName=FadeBasedOnViewAngle}) {materialFunction=&FadeBasedOnViewAngle}
  Reroute(None=LinearInterpolate(
    A=Multiply(A=@n85, B=@n86.Result)
    B=Multiply(A=@n87, B=@n88.Result)
    Alpha=ScalarParameter() {group=Digital_glow, parameterName=Use_simple_23Glow}
   ))
 ]
}
