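# Anim Blueprint for the Majesty Male pickaxe that makes its seven floating spike bones drift and lag.
# Each update, Random Vector Function turns scaled real time into per-spike offsets, applied as additive bone-space translations.
# A damped world-space AnimDynamics pass follows, blended in only in the frontend, then cached as FrontendDynamics.
# A second, always-on AnimDynamics pass adds linear-spring motion with locked angular limits before output.
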AnimBlueprint /BRCosmetics/Weapons/FORT_Melee/Pickaxe_Majesty_Male/Meshes/Majesty_Male_AnimBP
target_skeleton=&/Majesty_Male_Skeleton
graphs [
 AnimGraph {nodes [
   n1 = Anim Dynamics - Bone: dyn_floatingSpike_1(ComponentPose=Local To Component(LocalPose=Use cached pose 'FrontendDynamics'()), Alpha=1, ExternalForce="0,0,0") {
    bGravityOverrideInSimSpace=true
    bLinearSpring=true
    gravityScale=0.07
    linearSpringConstant=10
    alphaBoolBlend {blendOption=Linear}
    alphaScaleBiasClamp {scale=0.11}
    boundBone {boneName=dyn_floatingSpike_1}
    componentLinearAccScale=<0.11 0.13 0.18>
    physicsBodyDefinitions [{
      sphereCollisionRadius=0
      boundBone {boneName=dyn_floatingSpike_1}
      boxExtents=<1 1 1>
      constraintSetup {angularLimitsMax=<0 0 0>, angularLimitsMin=<0 0 0>, angularTarget=<0 0 0>, linearAxesMax=<30 40 30>, linearAxesMin=<-60 0 -60>}
      localJointOffset=<0 0 0>
     }]
   }
   n2 = Anim Dynamics - Bone: dyn_floatingSpike_2(ComponentPose=@n1, Alpha=1, ExternalForce="0,0,0") {
    bGravityOverrideInSimSpace=true
    bLinearSpring=true
    gravityScale=0.07
    linearSpringConstant=10
    alphaBoolBlend {blendOption=Linear}
    alphaScaleBiasClamp {scale=0.11}
    boundBone {boneName=dyn_floatingSpike_2}
    componentLinearAccScale=<0.11 0.13 0.18>
    physicsBodyDefinitions [{
      sphereCollisionRadius=0
      boundBone {boneName=dyn_floatingSpike_2}
      boxExtents=<1 1 1>
      constraintSetup {angularLimitsMax=<0 0 0>, angularLimitsMin=<0 0 0>, angularTarget=<0 0 0>, linearAxesMax=<30 30 30>, linearAxesMin=<-60 -60 -60>}
      localJointOffset=<0 0 0>
     }]
   }
   n3 = Anim Dynamics - Bone: dyn_floatingSpike_3(ComponentPose=@n2, Alpha=1, ExternalForce="0,0,0") {
    bGravityOverrideInSimSpace=true
    bLinearSpring=true
    gravityScale=0.07
    linearSpringConstant=10
    alphaBoolBlend {blendOption=Linear}
    alphaScaleBiasClamp {scale=0.11}
    boundBone {boneName=dyn_floatingSpike_3}
    componentLinearAccScale=<0.11 0.13 0.2>
    physicsBodyDefinitions [{
      sphereCollisionRadius=0
      boundBone {boneName=dyn_floatingSpike_3}
      boxExtents=<1 1 1>
      constraintSetup {angularLimitsMax=<0 0 0>, angularLimitsMin=<0 0 0>, angularTarget=<0 0 0>, linearAxesMax=<30 30 30>, linearAxesMin=<-60 -60 -60>}
      localJointOffset=<0 0 0>
     }]
   }
   n4 = Anim Dynamics - Bone: dyn_floatingSpike_4(ComponentPose=@n3, Alpha=1, ExternalForce="0,0,0") {
    bGravityOverrideInSimSpace=true
    bLinearSpring=true
    gravityScale=0.07
    linearSpringConstant=10
    alphaBoolBlend {blendOption=Linear}
    alphaScaleBiasClamp {scale=0.11}
    boundBone {boneName=dyn_floatingSpike_4}
    componentLinearAccScale=<0.11 0.13 0.18>
    physicsBodyDefinitions [{
      sphereCollisionRadius=0
      boundBone {boneName=dyn_floatingSpike_4}
      boxExtents=<1 1 1>
      constraintSetup {angularLimitsMax=<0 0 0>, angularLimitsMin=<0 0 0>, angularTarget=<0 0 0>, linearAxesMax=<30 30 30>, linearAxesMin=<-60 -60 -60>}
      localJointOffset=<0 0 0>
     }]
   }
   n5 = Anim Dynamics - Bone: dyn_floatingSpike_5(ComponentPose=@n4, Alpha=1, ExternalForce="0,0,0") {
    bGravityOverrideInSimSpace=true
    bLinearSpring=true
    gravityScale=0.07
    linearSpringConstant=10
    alphaBoolBlend {blendOption=Linear}
    alphaScaleBiasClamp {scale=0.11}
    boundBone {boneName=dyn_floatingSpike_5}
    componentLinearAccScale=<0.11 0.13 0.18>
    physicsBodyDefinitions [{
      sphereCollisionRadius=0
      boundBone {boneName=dyn_floatingSpike_5}
      boxExtents=<1 1 1>
      constraintSetup {angularLimitsMax=<0 0 0>, angularLimitsMin=<0 0 0>, angularTarget=<0 0 0>, linearAxesMax=<30 40 30>, linearAxesMin=<-60 0 -60>}
      localJointOffset=<0 0 0>
     }]
   }
   n6 = Anim Dynamics - Bone: dyn_floatingSpike_6(ComponentPose=@n5, Alpha=1, ExternalForce="0,0,0") {
    bGravityOverrideInSimSpace=true
    bLinearSpring=true
    gravityScale=0.07
    linearSpringConstant=10
    alphaBoolBlend {blendOption=Linear}
    alphaScaleBiasClamp {scale=0.11}
    boundBone {boneName=dyn_floatingSpike_6}
    componentLinearAccScale=<0.11 0.13 0.18>
    physicsBodyDefinitions [{
      sphereCollisionRadius=0
      boundBone {boneName=dyn_floatingSpike_6}
      boxExtents=<1 1 1>
      constraintSetup {angularLimitsMax=<0 0 0>, angularLimitsMin=<0 0 0>, angularTarget=<0 0 0>, linearAxesMax=<30 30 30>, linearAxesMin=<-60 -60 -60>}
      localJointOffset=<0 0 0>
     }]
   }
   n7 = Anim Dynamics - Bone: dyn_floatingSpike_7(ComponentPose=@n6, Alpha=1, ExternalForce="0,0,0") {
    bGravityOverrideInSimSpace=true
    bLinearSpring=true
    gravityScale=0.07
    linearSpringConstant=10
    alphaBoolBlend {blendOption=Linear}
    alphaScaleBiasClamp {scale=0.11}
    boundBone {boneName=dyn_floatingSpike_7}
    componentLinearAccScale=<0.11 0.13 0.21>
    physicsBodyDefinitions [{
      sphereCollisionRadius=0
      boundBone {boneName=dyn_floatingSpike_7}
      boxExtents=<1 1 1>
      constraintSetup {angularLimitsMax=<0 0 0>, angularLimitsMin=<0 0 0>, angularTarget=<0 0 0>, linearAxesMax=<30 30 30>, linearAxesMin=<-60 -60 -60>}
      localJointOffset=<0 0 0>
     }]
   }
   Output Pose(Result=Component To Local(ComponentPose=@n7)) {layerGroup=None}
   n8 = "Transform (Modify) Bone - Bone: dyn_floatingSpike_1"(
    ComponentPose=Local To Component(LocalPose=Input Pose())
    Alpha=0.5
    Translation=Get sine1() -> (sine1:Vector)
    Rotation="0,0,0"
    Scale="1,1,1"
   ) {
    alpha=0.5
    rotationMode=BMM_Additive
    translationMode=BMM_Additive
    translationSpace=BCS_BoneSpace
    alphaBoolBlend {blendOption=Linear}
    boneToModify {boneName=dyn_floatingSpike_1}
   }
   n9 = "Transform (Modify) Bone - Bone: dyn_floatingSpike_2"(ComponentPose=@n8, Alpha=1.1, Translation=Get sine3() -> (sine3:Vector), Rotation="0,0,0", Scale="1,1,1") {
    rotationMode=BMM_Additive
    translationMode=BMM_Additive
    translationSpace=BCS_BoneSpace
    alphaBoolBlend {blendOption=Linear}
    boneToModify {boneName=dyn_floatingSpike_2}
   }
   n10 = "Transform (Modify) Bone - Bone: dyn_floatingSpike_3"(ComponentPose=@n9, Alpha=1, Translation=Get sine5() -> (sine5:Vector), Rotation="0,0,0", Scale="1,1,1") {
    rotationMode=BMM_Additive
    translationMode=BMM_Additive
    translationSpace=BCS_BoneSpace
    alphaBoolBlend {blendOption=Linear}
    boneToModify {boneName=dyn_floatingSpike_3}
   }
   n11 = "Transform (Modify) Bone - Bone: dyn_floatingSpike_4"(ComponentPose=@n10, Alpha=0.7, Translation=Get sine4() -> (sine4:Vector), Rotation="0,0,0", Scale="1,1,1") {
    rotationMode=BMM_Additive
    translationMode=BMM_Additive
    translationSpace=BCS_BoneSpace
    alphaBoolBlend {blendOption=Linear}
    boneToModify {boneName=dyn_floatingSpike_4}
   }
   n12 = "Transform (Modify) Bone - Bone: dyn_floatingSpike_5"(ComponentPose=@n11, Alpha=0.8, Translation=Get sine7() -> (sine7:Vector), Rotation="0,0,0", Scale="1,1,1") {
    rotationMode=BMM_Additive
    translationMode=BMM_Additive
    translationSpace=BCS_BoneSpace
    alphaBoolBlend {blendOption=Linear}
    boneToModify {boneName=dyn_floatingSpike_5}
   }
   n13 = "Transform (Modify) Bone - Bone: dyn_floatingSpike_6"(ComponentPose=@n12, Alpha=1.6, Translation=Get sine2() -> (sine2:Vector), Rotation="0,0,0", Scale="1,1,1") {
    rotationMode=BMM_Additive
    translationMode=BMM_Additive
    translationSpace=BCS_BoneSpace
    alphaBoolBlend {blendOption=Linear}
    boneToModify {boneName=dyn_floatingSpike_6}
   }
   n14 = "Transform (Modify) Bone - Bone: dyn_floatingSpike_7"(ComponentPose=@n13, Alpha=0.7, Translation=Get sine6() -> (sine6:Vector), Rotation="0,0,0", Scale="1,1,1") {
    rotationMode=BMM_Additive
    translationMode=BMM_Additive
    translationSpace=BCS_BoneSpace
    alphaBoolBlend {blendOption=Linear}
    boneToModify {boneName=dyn_floatingSpike_7}
   }
   n15 = "Transform (Modify) Bone - Bone: dyn_floatingSpike_5"(ComponentPose=@n14, Alpha=1, Translation="0,3,1", Rotation="0,0,0", Scale="1,1,1") {
    translationMode=BMM_Additive
    translationSpace=BCS_BoneSpace
    alphaBoolBlend {blendOption=Linear}
    boneToModify {boneName=dyn_floatingSpike_5}
    translation {y=3, z=1}
   }
   n16 = Get frontendAlpha() -> (frontendAlpha:double)
   n17 = Anim Dynamics - Bone: dyn_floatingSpike_1(ComponentPose=@n15, Alpha=@n16, ExternalForce="0,0,0") {
    alpha=0.3
    bGravityOverrideInSimSpace=true
    bLinearSpring=true
    bOverrideLinearDamping=true
    gravityScale=0.01
    linearDampingOverride=0.95
    linearSpringConstant=30
    simulationSpace=World
    alphaBoolBlend {blendOption=Linear}
    alphaScaleBiasClamp {scale=0.05}
    boundBone {boneName=dyn_floatingSpike_1}
    componentLinearAccScale=<0.11 0.13 0.18>
    componentLinearVelScale=<0.1 0.1 0.1>
    physicsBodyDefinitions [{
      sphereCollisionRadius=0
      boundBone {boneName=dyn_floatingSpike_1}
      boxExtents=<1 1 1>
      constraintSetup {angularLimitsMax=<0 0 0>, angularLimitsMin=<0 0 0>, angularTarget=<0 0 0>, linearAxesMax=<30 80 30>, linearAxesMin=<-60 0 -60>}
      localJointOffset=<0 0 0>
     }]
   }
   n18 = Anim Dynamics - Bone: dyn_floatingSpike_2(ComponentPose=@n17, Alpha=@n16, ExternalForce="0,0,0") {
    alpha=0.3
    bGravityOverrideInSimSpace=true
    bLinearSpring=true
    bOverrideLinearDamping=true
    gravityScale=0.01
    linearDampingOverride=0.95
    linearSpringConstant=30
    simulationSpace=World
    alphaBoolBlend {blendOption=Linear}
    alphaScaleBiasClamp {scale=0.05}
    boundBone {boneName=dyn_floatingSpike_2}
    componentLinearAccScale=<0.11 0.13 0.18>
    componentLinearVelScale=<0.1 0.1 0.1>
    physicsBodyDefinitions [{
      sphereCollisionRadius=0
      boundBone {boneName=dyn_floatingSpike_2}
      boxExtents=<1 1 1>
      constraintSetup {angularLimitsMax=<0 0 0>, angularLimitsMin=<0 0 0>, angularTarget=<0 0 0>, linearAxesMax=<30 80 30>, linearAxesMin=<-60 -60 -60>}
      localJointOffset=<0 0 0>
     }]
   }
   n19 = Anim Dynamics - Bone: dyn_floatingSpike_3(ComponentPose=@n18, Alpha=@n16, ExternalForce="0,0,0") {
    alpha=0.3
    bGravityOverrideInSimSpace=true
    bLinearSpring=true
    bOverrideLinearDamping=true
    gravityScale=0.01
    linearDampingOverride=0.95
    linearSpringConstant=30
    simulationSpace=World
    alphaBoolBlend {blendOption=Linear}
    alphaScaleBiasClamp {scale=0.05}
    boundBone {boneName=dyn_floatingSpike_3}
    componentLinearAccScale=<0.11 0.13 0.2>
    componentLinearVelScale=<0.1 0.1 0.1>
    physicsBodyDefinitions [{
      sphereCollisionRadius=0
      boundBone {boneName=dyn_floatingSpike_3}
      boxExtents=<1 1 1>
      constraintSetup {angularLimitsMax=<0 0 0>, angularLimitsMin=<0 0 0>, angularTarget=<0 0 0>, linearAxesMax=<30 80 30>, linearAxesMin=<-60 -60 -60>}
      localJointOffset=<0 0 0>
     }]
   }
   n20 = Anim Dynamics - Bone: dyn_floatingSpike_4(ComponentPose=@n19, Alpha=@n16, ExternalForce="0,0,0") {
    alpha=0.3
    bGravityOverrideInSimSpace=true
    bLinearSpring=true
    bOverrideLinearDamping=true
    gravityScale=0.01
    linearDampingOverride=0.95
    linearSpringConstant=30
    simulationSpace=World
    alphaBoolBlend {blendOption=Linear}
    alphaScaleBiasClamp {scale=0.05}
    boundBone {boneName=dyn_floatingSpike_4}
    componentLinearAccScale=<0.11 0.13 0.18>
    componentLinearVelScale=<0.1 0.1 0.1>
    physicsBodyDefinitions [{
      sphereCollisionRadius=0
      boundBone {boneName=dyn_floatingSpike_4}
      boxExtents=<1 1 1>
      constraintSetup {angularLimitsMax=<0 0 0>, angularLimitsMin=<0 0 0>, angularTarget=<0 0 0>, linearAxesMax=<30 80 30>, linearAxesMin=<-60 -60 -60>}
      localJointOffset=<0 0 0>
     }]
   }
   n21 = Anim Dynamics - Bone: dyn_floatingSpike_5(ComponentPose=@n20, Alpha=@n16, ExternalForce="0,0,0") {
    alpha=0.3
    bGravityOverrideInSimSpace=true
    bLinearSpring=true
    bOverrideLinearDamping=true
    gravityScale=0.01
    linearDampingOverride=0.95
    linearSpringConstant=30
    simulationSpace=World
    alphaBoolBlend {blendOption=Linear}
    alphaScaleBiasClamp {scale=0.05}
    boundBone {boneName=dyn_floatingSpike_5}
    componentLinearAccScale=<0.11 0.13 0.18>
    componentLinearVelScale=<0.1 0.1 0.1>
    physicsBodyDefinitions [{
      sphereCollisionRadius=0
      boundBone {boneName=dyn_floatingSpike_5}
      boxExtents=<1 1 1>
      constraintSetup {angularLimitsMax=<0 0 0>, angularLimitsMin=<0 0 0>, angularTarget=<0 0 0>, linearAxesMax=<30 80 30>, linearAxesMin=<-60 0 -60>}
      localJointOffset=<0 0 0>
     }]
   }
   n22 = Anim Dynamics - Bone: dyn_floatingSpike_6(ComponentPose=@n21, Alpha=@n16, ExternalForce="0,0,0") {
    alpha=0.3
    bGravityOverrideInSimSpace=true
    bLinearSpring=true
    bOverrideLinearDamping=true
    gravityScale=0.01
    linearDampingOverride=0.95
    linearSpringConstant=30
    simulationSpace=World
    alphaBoolBlend {blendOption=Linear}
    alphaScaleBiasClamp {scale=0.05}
    boundBone {boneName=dyn_floatingSpike_6}
    componentLinearAccScale=<0.11 0.13 0.18>
    componentLinearVelScale=<0.1 0.1 0.1>
    physicsBodyDefinitions [{
      sphereCollisionRadius=0
      boundBone {boneName=dyn_floatingSpike_6}
      boxExtents=<1 1 1>
      constraintSetup {angularLimitsMax=<0 0 0>, angularLimitsMin=<0 0 0>, angularTarget=<0 0 0>, linearAxesMax=<30 80 30>, linearAxesMin=<-60 -60 -60>}
      localJointOffset=<0 0 0>
     }]
   }
   n23 = Anim Dynamics - Bone: dyn_floatingSpike_7(ComponentPose=@n22, Alpha=@n16, ExternalForce="0,0,0") {
    alpha=0.3
    bGravityOverrideInSimSpace=true
    bLinearSpring=true
    bOverrideLinearDamping=true
    gravityScale=0.01
    linearDampingOverride=0.95
    linearSpringConstant=30
    simulationSpace=World
    alphaBoolBlend {blendOption=Linear}
    alphaScaleBiasClamp {scale=0.05}
    boundBone {boneName=dyn_floatingSpike_7}
    componentLinearAccScale=<0.11 0.13 0.21>
    componentLinearVelScale=<0.1 0.1 0.1>
    physicsBodyDefinitions [{
      sphereCollisionRadius=0
      boundBone {boneName=dyn_floatingSpike_7}
      boxExtents=<1 1 1>
      constraintSetup {angularLimitsMax=<0 0 0>, angularLimitsMin=<0 0 0>, angularTarget=<0 0 0>, linearAxesMax=<30 80 30>, linearAxesMin=<-60 -60 -60>}
      localJointOffset=<0 0 0>
     }]
   }
   Save cached pose 'FrontendDynamics'(Pose=Component To Local(ComponentPose=@n23)) {cacheName=FrontendDynamics}
  ]}
 EventGraph {nodes [
   Event BlueprintUpdateAnimation()
   > n1 = Sequence!(execute) -> (then_0, then_1, then_2:Exec, then_3:Exec, then_4:Exec, then_5:Exec, then_6:Exec, then_7:Exec, then_8:Exec)
   n2 = GetOwnerSkeletalMeshComponent()
   > n3 = Random Vector Function!(
    execute
    CharacterPartAnimInstance:Custom Character Part Anim Instance Object Reference
    SkelMeshComponent:Mesh Component Object Reference=@n2
    JointName:Name=None
    GameTimeInSeconds:double=Get customSpeed() -> (customSpeed:double)
    TimeOffset:double=0
    PerlinScale:double=1
    Frequency:Vector="0.6,0.6,0.6"
    Amplitude:Vector="1.8,1.8,2"
    Offset:Vector="3,3,2"
   ) -> (then, VectorOut:Vector)
   > Set sine1(sine1:Vector=@n3) -> (Output_Get:Vector)
   n4 = Random Vector Function(
    execute=@n1.then_1
    SkelMeshComponent=@n2
    JointName=None
    GameTimeInSeconds=Get customSpeed()
    TimeOffset=0
    PerlinScale=1
    Frequency="0.35,0.35,0.35"
    Amplitude="3,3,2"
    Offset="-5,-5,-3"
   )
   > Set sine2(sine2:Vector=@n4) -> (Output_Get:Vector)
   n5 = Random Vector Function(
    execute=@n1.then_2
    SkelMeshComponent=@n2
    JointName=None
    GameTimeInSeconds=Get customSpeed()
    TimeOffset=0
    PerlinScale=1
    Frequency="0.4,0.4,0.4"
    Amplitude="1,2.3,2.3"
    Offset="1,2,2"
   )
   > Set sine3(sine3:Vector=@n5) -> (Output_Get:Vector)
   n6 = Random Vector Function(
    execute=@n1.then_3
    SkelMeshComponent=@n2
    JointName=None
    GameTimeInSeconds=Get customSpeed()
    TimeOffset=0
    PerlinScale=1
    Frequency="0.7,0.7,0.7"
    Amplitude="2.4,2.4,2.4"
    Offset="4,4,4"
   )
   > Set sine4(sine4:Vector=@n6) -> (Output_Get:Vector)
   n7 = Random Vector Function(
    execute=@n1.then_4
    SkelMeshComponent=@n2
    JointName=None
    GameTimeInSeconds=Get customSpeed()
    TimeOffset=0
    PerlinScale=1
    Frequency="0.4,0.4,0.4"
    Amplitude="3.5,3.5,3.5"
    Offset="3,3,3"
   )
   > Set sine5(sine5:Vector=@n7) -> (Output_Get:Vector)
   n8 = Random Vector Function(
    execute=@n1.then_5
    SkelMeshComponent=@n2
    JointName=None
    GameTimeInSeconds=Get customSpeed()
    TimeOffset=0
    PerlinScale=1
    Frequency="0.7,0.7,0.7"
    Amplitude="3.5,3.5,3.5"
    Offset="2,2,2"
   )
   > Set sine6(sine6:Vector=@n8) -> (Output_Get:Vector)
   n9 = Random Vector Function(
    execute=@n1.then_6
    SkelMeshComponent=@n2
    JointName=None
    GameTimeInSeconds=Get customSpeed()
    TimeOffset=0
    PerlinScale=1
    Frequency="0.5,0.5,0.5"
    Amplitude="2,2,2"
    Offset="0,0,0"
   )
   > Set sine7(sine7:Vector=@n9) -> (Output_Get:Vector)
   Set customSpeed(execute=@n1.then_7, customSpeed:double=float / float(A=GetRealTimeSeconds(), B=1.25) {is=title K2Node_CallFunction}) -> (Output_Get:double)
   Set frontendAlpha(execute=@n1.then_8, frontendAlpha:double=SelectFloat(A=0.51, B=0, bPickA=Get bIsInFrontEnd())) -> (Output_Get:double)
   TryGetPawnOwner()
  ]}
 NewEventGraph_0 {links [], nodes []}
]
variables [
 customSpeed {type=number}
 frontendAlpha {type=number}
 Pawn {type=FortPlayerPawn}
 rotator1 {type=Rotator}
 Self_0 {type=Majesty_Male_AnimBP_C, default=null}
 sine1 {type=Vector}
 sine2 {type=Vector}
 sine3 {type=Vector}
 sine4 {type=Vector}
 sine5 {type=Vector}
 sine6 {type=Vector}
 sine7 {type=Vector}
 SkelMesh {type=SkeletalMesh, default=null}
]
